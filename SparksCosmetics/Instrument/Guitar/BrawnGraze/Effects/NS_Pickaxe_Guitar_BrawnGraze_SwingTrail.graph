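# Niagara swing trail for the BrawnGraze guitar pickaxe, built from two short-lived ribbon emitters, an accent layer and a colored layer.
# Both use shared AnimTrail modules to spawn, initialize and update ribbons from bone or socket indices, then set dynamic material parameters.
# Brightness scales by Fortnite time-of-day curves and alpha fades over a curve, while the colored layer takes its color from a system time-of-day color.
# Ribbon-width scaling and added velocity are disabled.
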NiagaraSystem /SparksCosmetics/Instrument/Guitar/BrawnGraze/Effects/NS_Pickaxe_Guitar_BrawnGraze_SwingTrail
dependencies {
 usedDataInterfaces [&NiagaraDataInterfaceCurve, &NiagaraDataInterfaceColorCurve]
 usedDynamicInputs [&ColorFromCurve]
 usedModules [
  &self:SystemSpawnScript.SystemScriptSource.SystemScriptGraph.NiagaraNodeAssignment_2.SetVariables_#1
  &SystemState
  &self:SystemSpawnScript.SystemScriptSource.SystemScriptGraph.NiagaraNodeAssignment_0.SetVariables_#2
  &self:SystemSpawnScript.SystemScriptSource.SystemScriptGraph.NiagaraNodeAssignment_1.SetVariables_#3
 ]
}
emitters {
 Trail_Accent {
  emitterUpdateScript [
   &EmitterState()
   &FN_SpawnAnimTrail_v2()
  ]
  particleSpawnScript [
   /Game/Effects/Niagara/Modules/AnimTrail/FN_InitializeAnimTrail/FN_InitializeAnimTrailBoneIndecies(First Bone or Socket Index=1)
   /Game/Effects/Niagara/Modules/AnimTrail/FN_InitializeAnimTrail/FN_InitializeAnimTrail(AnimTrailWidth=1, Lifetime=0.1)
   &AddVelocity(Velocity=<0 500 0>) {enabled=false}
  ]
  particleUpdateScript [
   &ScaleRibbonWidth(Ribbon Width Scale=&FloatFromCurve(FloatCurve=DataInterface() {propertyValues {Curve {keys [
         {arriveTangent=-1.2579, interpMode=RCIM_Cubic, time=0, value=1}
         {arriveTangent=-0.9664, interpMode=RCIM_Cubic, leaveTangent=-0.9664, tangentMode=RCTM_User, time=1, value=0}
        ]}}})) {enabled=false}
   &FN_UpdateAnimTrail_v2()
   &ScaleColor(
    Scale Alpha=&FloatFromCurve(
     FloatCurve=DataInterface() {propertyValues {Curve {keys [
         {interpMode=RCIM_Cubic, leaveTangent=-0.5998, tangentMode=RCTM_Break, time=0, value=1}
         {arriveTangent=-0.2139, interpMode=RCIM_Cubic, tangentMode=RCTM_Break, time=1, value=0}
        ]}}}
     Scale Curve=&FloatFromCurve(
      FloatCurve=DataInterface() {propertyValues {Curve {keys [
          <at 0 0.75 RCIM_Cubic>
          {arriveTangent=0.5638, interpMode=RCIM_Cubic, leaveTangent=0.5638, time=0.2025, value=0.85}
          {arriveTangentWeight=0.0858, interpMode=RCIM_Cubic, leaveTangentWeight=0.0558, tangentMode=RCTM_User, time=0.4434, value=1}
          {arriveTangentWeight=0.0572, interpMode=RCIM_Cubic, leaveTangentWeight=0.1462, tangentMode=RCTM_User, time=0.6, value=1}
          <at 1 0.75 RCIM_Cubic>
         ]}}}
      CurveIndex=Linked() {linkedVariable {name=NPC.FortniteNPC.TimeOfDay_0-1, type=NiagaraFloat}}
     )
    )
    ScaleA=true
    ScaleRGB=true
   ) {module=ScaleColor002}
   &DynamicMaterialParameters(
    Index 0 Param 4=1
    Index 1 Param 4=1
    Index 2 Param 1=15
    Index 2 Param 2=0.1929
    Index 2 Param 3=0.5
    Index 2 Param 4=0.6
    Module.Index0Float0Write=true
    Module.Index0Float1Write=true
    Module.Index0Float2Write=true
    Module.Index0Float3Write=true
    Module.Index1Float0Write=true
    Module.Index1Float1Write=true
    Module.Index1Float2Write=true
    Module.Index1Float3Write=true
    Module.Index2Float0Write=true
    Module.Index2Float1Write=true
    Module.Index2Float2Write=true
    Module.Index2Float3Write=true
    Write Parameter Index 1=::NewEnumerator1
    Write Parameter Index 2=::NewEnumerator1
   )
   &SolveForcesAndVelocity(Clamp Velocity=false, Limit Acceleration=false)
  ]
  rendererClasses [&NiagaraRibbonRendererProperties]
  renderers [{rendererClass=&NiagaraRibbonRendererProperties}]
 }
 Trail_Color001 {
  emitterUpdateScript [
   &EmitterState()
   &FN_SpawnAnimTrail_v2()
  ]
  particleSpawnScript [
   &FN_InitializeAnimTrailBoneIndecies(First Bone or Socket Index=1)
   &FN_InitializeAnimTrail(AnimTrailWidth=1, Color=$System.TOD Color, Lifetime=0.13)
   &AddVelocity(Velocity=<0 500 0>) {enabled=false}
  ]
  particleUpdateScript [
   &ScaleRibbonWidth(Ribbon Width Scale=&FloatFromCurve(FloatCurve=DataInterface() {propertyValues {Curve {keys [
         {arriveTangent=-1.2579, interpMode=RCIM_Cubic, time=0, value=1}
         {arriveTangent=-0.9664, interpMode=RCIM_Cubic, leaveTangent=-0.9664, tangentMode=RCTM_User, time=1, value=0}
        ]}}})) {enabled=false}
   &ScaleColor(
    Scale Alpha=&FloatFromCurve(
     FloatCurve=DataInterface() {propertyValues {Curve {keys [
         {interpMode=RCIM_Cubic, leaveTangent=-0.5729, tangentMode=RCTM_Break, time=0, value=1}
         {arriveTangent=-0.5809, interpMode=RCIM_Cubic, leaveTangent=-0.5809, tangentMode=RCTM_User, time=0.4, value=0.15}
         {arriveTangent=-0.2139, interpMode=RCIM_Cubic, tangentMode=RCTM_Break, time=1, value=0}
        ]}}}
     Scale Curve=&FloatFromCurve(
      FloatCurve=DataInterface() {propertyValues {Curve {keys [
          <at 0 0.85 RCIM_Cubic>
          {arriveTangent=0.3383, interpMode=RCIM_Cubic, leaveTangent=0.3383, time=0.2025, value=0.95}
          {arriveTangentWeight=0.0858, interpMode=RCIM_Cubic, leaveTangentWeight=0.0558, tangentMode=RCTM_User, time=0.4434, value=1}
          {arriveTangentWeight=0.0572, interpMode=RCIM_Cubic, leaveTangentWeight=0.1462, tangentMode=RCTM_User, time=0.6, value=1}
          <at 1 0.85 RCIM_Cubic>
         ]}}}
      CurveIndex=Linked() {linkedVariable {name=NPC.FortniteNPC.TimeOfDay_0-1, type=NiagaraFloat}}
     )
    )
    ScaleA=true
    ScaleRGB=true
   ) {module=ScaleColor002}
   &FN_UpdateAnimTrail_v2()
   &DynamicMaterialParameters(
    Index 0 Param 1=0.7
    Index 1 Param 1=3
    Index 2 Param 1=10
    Index 2 Param 2=0.5
    Index 2 Param 3=0.5
    Index 2 Param 4=0.6
    Module.Index0Float0Write=true
    Module.Index0Float1Write=true
    Module.Index0Float2Write=true
    Module.Index0Float3Write=true
    Module.Index1Float0Write=true
    Module.Index1Float1Write=true
    Module.Index1Float2Write=true
    Module.Index1Float3Write=true
    Module.Index2Float0Write=true
    Module.Index2Float1Write=true
    Module.Index2Float2Write=true
    Module.Index2Float3Write=true
    Write Parameter Index 1=::NewEnumerator1
    Write Parameter Index 2=::NewEnumerator1
   )
   &SolveForcesAndVelocity(Clamp Velocity=false, Limit Acceleration=false)
  ]
  rendererClasses [&NiagaraRibbonRendererProperties]
  renderers [{rendererClass=&NiagaraRibbonRendererProperties}]
 }
}
summary {
 systemName=NS_Pickaxe_Guitar_BrawnGraze_SwingTrail
 emitters [
  {emitterName=Trail_Accent, rendererClasses [&NiagaraRibbonRendererProperties]}
  {emitterName=Trail_Color001, rendererClasses [&NiagaraRibbonRendererProperties]}
 ]
 userVariables [
  User.RibbonIDOffset {type=NiagaraInt32}
  User.SkeletalMesh {
   type=NiagaraDataInterfaceSkeletalMesh
   defaultValue {
    struct=NiagaraExt_VariableValue_DataInterface
    value {dataInterface=&self:NiagaraDataInterfaceSkeletalMesh_0}
   }
  }
  User.TrailColor {type=LinearColor, description="", defaultValue {struct=LinearColor, value=<rgba 1 1 1 0.25>}}
  User.TrailSecondaryColor {type=LinearColor, description="", defaultValue {struct=LinearColor, value=<rgba 10 10 10 0.25>}}
 ]
}
