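# Physics asset for the BrawnGraze guitar pickaxe cosmetic, giving its skeletal mesh a small simulation setup.
# A sphere body on the dyn_hat bone and a capsule body on the handle bone are joined by one constraint.
# Tight swing and twist angle limits on that constraint restrict how far the hat can rotate relative to the handle.
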PhysicsAsset /SparksCosmetics/Instrument/Guitar/BrawnGraze/Meshes/BrawnGraze_Guitar_Pickaxe_Physics
bodies {
 dyn_hat [{shapeType=Sphere, radius=7.2927, shapeName=dyn_hat_sphere, center=<-1 0 0>}]
 handle [{shapeType=Capsule, length=5, radius=5, rotation=<rot -5.5616 0 0>}]
}
constraints [{bone1Name=dyn_hat, bone2Name=handle, swing1LimitDegrees=10, swing2LimitDegrees=22, twistLimitDegrees=12}]
