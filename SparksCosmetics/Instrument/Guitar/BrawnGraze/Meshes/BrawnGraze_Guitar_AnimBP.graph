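# Animation Blueprint for the BrawnGraze guitar cosmetic, adding physics-driven secondary motion to its dyn_hat bone.
# Three chained Anim Dynamics nodes simulate relative to the handle: one always enabled, then two tuned for normal play and skydiving, both disabled in the front end.
# A RigidBody node exists but does not reach the output pose.
# The event graph sets a skydive-dependent Gravity Override variable, while the dynamics nodes use constant gravity.
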AnimBlueprint /SparksCosmetics/Instrument/Guitar/BrawnGraze/Meshes/BrawnGraze_Guitar_AnimBP
target_skeleton=&/BrawnGraze_Guitar_Skeleton
graphs [
 AnimGraph {nodes [
   n1 = RigidBody() {
    bOverrideWorldGravity=true
    overridePhysicsAsset=&/BrawnGraze_Guitar_Pickaxe_Physics
    simulationSpace=BaseBoneSpace
    alphaBoolBlend {blendOption=Linear}
    baseBoneRef {boneName=handle}
    componentAppliedLinearAccClamp=<3000 3000 3000>
    componentLinearAccScale=<0.4 0.4 0.4>
    componentLinearVelScale=<0.2 0.2 0.2>
    overrideWorldGravity {z=-980}
    simSpaceSettings {
     maxAngularAcceleration=1000
     maxAngularVelocity=6
     maxLinearAcceleration=700
     maxLinearVelocity=600
     velocityScaleZ=0.1
     worldAlpha=1
     externalLinearDragV=<0.2 0.2 0.02>
    }
   }
   Component To Local(ComponentPose=@n1)
   n2 = Anim Dynamics - Bone: dyn_hat!(
    ComponentPose=Local To Component(LocalPose=Input Pose())
    bAlphaBoolEnabled:bool=True
    GravityOverride:Vector="0,0,-980"
    ExternalForce="0,0,0"
   ) -> (Pose) {
    alphaInputType=Bool
    angularDampingOverride=0.99
    angularSpringConstant=200
    bOverrideAngularDamping=true
    bUseGravityOverride=true
    linearDampingOverride=0.9
    simulationSpace=BoneRelative
    alphaBoolBlend {blendOption=Linear}
    boundBone {boneName=dyn_hat}
    componentAppliedLinearAccClamp=<3000 3000 3000>
    componentLinearAccScale=<0.25 0.25 0.25>
    componentLinearVelScale=<0.8 0.8 0.8>
    gravityOverride {z=-980}
    physicsBodyDefinitions [{
      boundBone {boneName=dyn_hat}
      boxExtents=<20 20 20>
      constraintSetup {angularLimitsMax=<5 20 5>, angularLimitsMin=<-4 -4 -5>, angularTarget=<3 0 0>}
      localJointOffset=<1 0 0>
     }]
    relativeSpaceBone {boneName=handle}
    simSpaceSettings {maxAngularAcceleration=40, maxAngularVelocity=4, simSpaceAngularAlpha=0.5}
   }
   n3 = Get bIsSkyDiving()
   n4 = Get bIsInFrontEnd()
   n5 = Anim Dynamics - Bone: dyn_hat(
    ComponentPose=@n2
    bAlphaBoolEnabled=AND Boolean(A=NOT Boolean(A=@n4), B=NOT Boolean(A=@n3))
    GravityOverride="0,0,-980"
    ExternalForce="0,0,0"
   ) {
    alphaInputType=Bool
    angularDampingOverride=0.99
    angularSpringConstant=200
    bOverrideAngularDamping=true
    bUseGravityOverride=true
    linearDampingOverride=0.9
    simulationSpace=BoneRelative
    alphaBoolBlend {blendOption=Linear}
    boundBone {boneName=dyn_hat}
    componentAppliedLinearAccClamp=<3000 3000 3000>
    componentLinearAccScale=<0.4 0.4 0.4>
    componentLinearVelScale=<0.8 0.8 0.8>
    gravityOverride {z=-980}
    physicsBodyDefinitions [{
      boundBone {boneName=dyn_hat}
      boxExtents=<15 15 15>
      constraintSetup {angularLimitsMax=<16 12 8>, angularLimitsMin=<-4 -4 -5>, angularTarget=<3 0 0>, linearAxesMax=<0 0 4>}
      localJointOffset=<1 0 0>
     }]
    relativeSpaceBone {boneName=handle}
    simSpaceSettings {maxAngularAcceleration=50, maxAngularVelocity=4, simSpaceAngularAlpha=0.4}
   }
   n6 = Get bIsInFrontEnd()
   n7 = Anim Dynamics - Bone: dyn_hat(
    ComponentPose=@n5
    bAlphaBoolEnabled=AND Boolean(A=NOT Boolean(A=@n6), B=Get bIsSkyDiving())
    GravityOverride="0,0,-980"
    ExternalForce="0,0,0"
   ) {
    alphaInputType=Bool
    angularDampingOverride=0.99
    angularSpringConstant=200
    bOverrideAngularDamping=true
    bUseGravityOverride=true
    linearDampingOverride=0.9
    simulationSpace=BoneRelative
    alphaBoolBlend {blendOption=Linear}
    boundBone {boneName=dyn_hat}
    componentAppliedLinearAccClamp=<3000 3000 3000>
    componentLinearAccScale=<0.4 0.4 0.4>
    componentLinearVelScale=<0.8 0.8 0.8>
    gravityOverride {z=-980}
    physicsBodyDefinitions [{
      boundBone {boneName=dyn_hat}
      boxExtents=<15 15 15>
      constraintSetup {angularLimitsMax=<16 12 8>, angularLimitsMin=<-4 -10 -5>, angularTarget=<3 0 0>}
      localJointOffset=<1 0 0>
     }]
    relativeSpaceBone {boneName=handle}
    simSpaceSettings {maxAngularAcceleration=50, maxAngularVelocity=4, simSpaceAngularAlpha=0.4}
   }
   Output Pose(Result=Component To Local(ComponentPose=@n7))
  ]}
 EventGraph {nodes [
   Event BlueprintUpdateAnimation()
   > n1 = Branch(Condition=OR Boolean(A=Get bIsSkydiveFloating(), B=Get bIsSkyDiving()))
   > Set Gravity Override(Gravity Override="0,0,-900")
   Set Gravity Override(execute=@n1.else, Gravity Override="0,0,-980")
   TryGetPawnOwner()
  ]}
]
variables [Gravity Override {type=Vector}]
